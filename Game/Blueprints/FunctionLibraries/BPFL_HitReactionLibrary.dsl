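; Blueprint function library of shared helpers for Fort pawn hit reactions and control states such as stagger, knockback and stun.
; It converts gameplay cue or damage parameters into gameplay event data, which feeds proxy-asset montage selection.
; Additive reactions play as local montages, throttled by minimum interval.
; Control propagates to bumped pawns with an ability system through an Event.Control.Propagation event, tagged with the original pawn's stagger or knockback state.
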(fn ActorHasFortAIAnimationComponent
  (params (Actor :type Actor%20Object%20Reference))
  (Utilities|FlowControl|Sequence
    (:then_0
      (bind _actor (Variables|GetActor)
        (:then
          (Utilities|IsValid (Actor|GetComponentbyClass _actor "/Script/FortniteAI.FortAIComponent_Animation")
            (:Is%20Valid
              (return true (Actor|GetComponentbyClass _actor "/Script/FortniteAI.FortAIComponent_Animation")))) ; Do we have the Fort AI Animation Component?
        )
        (:else)) ; Actor we fed-in is Valid?
    )
    (:then_1
      (return false 0))))

(fn AttemptToPlayAdditiveHitReactionLocalMontage
  (params (Montage :type Anim%20Montage%20Object%20Reference) (Section :type Name) (FortPawn :type Fort%20Pawn%20Object%20Reference) (LastAdditiveHitReactionStartTime :type Float%20%28double-precision%29) (MinTimeBetweenAdditiveHitReactions :type Float%20%28double-precision%29))
  (Utilities|FlowControl|Sequence
    (:then_0
      (Variables|GetMontage
        (:then
          (Variables|GetFortPawn
            (:then
              (if (> (- (Utilities|Time|GetGameTimeinSeconds 0) (Variables|GetLastAdditiveHitReactionStartTime)) (Variables|GetMinTimeBetweenAdditiveHitReactions))
                (bind (_success _lastadditivehitreactstarttime) (CallFunction|PlayAdditiveHitReactionLocalMontage (Variables|GetMontage) (Variables|GetSection) (Variables|GetFortPawn)))
                (return _success _lastadditivehitreactstarttime)))
            (:else)))
        (:else)))
    (:then_1
      (return false -1.0))))

(fn AttemptToPropagateControlInfoToBumpedPawn
  (params (GameplayEventDataFromOriginalHitReaction :type Gameplay%20Event%20Data%20Structure) (OriginalPawnDoingHitReaction :type Actor%20Object%20Reference) (BumpedPawnWeAreAttemptingToPropagateControlInfoTo :type Actor%20Object%20Reference) (BumpNormalImpulse :type Vector) (BumpHitResult :type Hit%20Result%20Structure))
  (bind (_eventtag _instigator _target _optionalobject _optionalobject2 _contexthandle _instigatortags _targettags _eventmagnitude) (Utilities|Struct|BreakGameplayEventData (Variables|GetGameplayEventDataFromOriginalHitReaction)))
  (bind (_eventtag_2 _instigator_2 _target_2 _optionalobject_2 _optionalobject2_2 _contexthandle_2 _instigatortags_2) (Utilities|Struct|BreakGameplayEventData (Variables|GetGameplayEventDataFromOriginalHitReaction)))
  (Utilities|FlowControl|Sequence
    (:then_0
      (Variables|GetOriginalPawnDoingHitReaction
        (:then
          (bind _bumpedpawnweareattemptingtopropagatecontrolinfoto (Variables|GetBumpedPawnWeAreAttemptingToPropagateControlInfoTo)
            (:then
              (Utilities|IsValid (Ability|GetAbilitySystemComponent _bumpedpawnweareattemptingtopropagatecontrolinfoto)
                (:Is%20Valid
                  (Utilities|FlowControl|Sequence
                    (:then_0
                      (bind _output_get (LocalVariables|SetInstigatorTags _instigatortags_2))
                      (GameplayTags|AddGameplayTag _output_get "(TagName=\"Gameplay.Status.Control.Propagation\")" :!class /Script/GameplayTags.BlueprintGameplayTagLibrary) ; Add Control Propagation Tag so we KNOW\non the Anim Chooser Table side that it's a\nPropagated Hit Reaction we are reacting to
                      (bind _asfort_pawn (Utilities|Casting|CastToFortPawn (Variables|GetOriginalPawnDoingHitReaction)))
                      (GameplayTags|AppendGameplayTagContainers (LocalVariables|GetInstigatorTags) (select (.IsStaggered%20%28FullBodyHitReact%29 (CallFunction|IsFortPawnBeingControlled _asfort_pawn)) (GameplayTags|MakeLiteralGameplayTagContainer "(GameplayTags=((TagName=\"Gameplay.Status.FullBodyHit\")))") (select (.IsKnockedback (CallFunction|IsFortPawnBeingControlled _asfort_pawn)) (GameplayTags|MakeLiteralGameplayTagContainer "(GameplayTags=((TagName=\"Gameplay.Status.Knockback\")))") (GameplayTags|MakeLiteralGameplayTagContainer)))) ; Append the Control status (FBHR or KB) to the Instigator Tags\nso on the Anim Chooser Table side, we can decide how to react.\n\nSince we are Propagating this info from an Attacker, through a \nControlled Pawn, and to another Pawn, we need to append the\nstatus of that Controlled Pawn as additional context.
                    )
                    (:then_1
                      (Ability|SendGameplayEventtoActor (Variables|GetBumpedPawnWeAreAttemptingToPropagateControlInfoTo) (GameplayTags|MakeLiteralGameplayTag "(TagName=\"Event.Control.Propagation\")") (Utilities|Struct|MakeGameplayEventData (GameplayTags|MakeLiteralGameplayTag "(TagName=\"Event.Control.Propagation\")") _instigator (Variables|GetBumpedPawnWeAreAttemptingToPropagateControlInfoTo) _optionalobject (Variables|GetOriginalPawnDoingHitReaction) _contexthandle (LocalVariables|GetInstigatorTags) _targettags _eventmagnitude (Ability|TargetData|AbilityTargetDatafromHitResult (Variables|GetBumpHitResult))) :!pinned true) ; This sends the Impact Damage through the Event Magnitude\nThe Impulse Data (KB and angle) is in the Context Handle\n\nThe Pawn being Propagated TO needs an Ability that reacts to\nthe Gameplay Event tag, takes this data, and decides whether to\napply it, and how much of it to apply and Propagate forward.
                      (return true (Variables|GetBumpedPawnWeAreAttemptingToPropagateControlInfoTo)))))))
            (:else)))
        (:else)))
    (:then_1
      (return false 0))))

(fn ConvertGameplayCueParametersToGameplayEventData
  (params (GameplayCueParameters :type Gameplay%20Cue%20Parameters%20Structure) (DamagedActor :type Actor%20Object%20Reference))
  (bind (_normalizedmagnitude _rawmagnitude _effectcontext _matchedtagname _originaltag _aggregatedsourcetags _aggregatedtargettags _location _normal _instigator) (Ability|GameplayCue|BreakGameplayCueParameters (Variables|GetGameplayCueParameters)))
  (return (Utilities|Struct|MakeGameplayEventData _matchedtagname _instigator (Variables|GetDamagedActor) 0 0 _effectcontext _aggregatedsourcetags _aggregatedtargettags _rawmagnitude (Ability|TargetData|AbilityTargetDatafromHitResult (Ability|EffectContext|GetHitResult _effectcontext)))))

(fn ConvertOnDamagePlayEffectsToGameplayEventData
  (params (Damage :type Float%20%28double-precision%29) (DamageTags :type Gameplay%20Tag%20Container%20Structure) (Momentum :type Vector) (HitInfo :type Hit%20Result%20Structure) (InstigatedBy :type Fort%20Pawn%20Object%20Reference) (DamageCauser :type Actor%20Object%20Reference) (EffectContext :type Gameplay%20Effect%20Context%20Handle%20Structure) (DamagedActor :type Actor%20Object%20Reference))
  (return (Utilities|Struct|MakeGameplayEventData "(TagName=\"\")" (Variables|GetInstigatedBy) (Variables|GetDamagedActor) 0 0 (Variables|GetEffectContext) (GameplayTags|GetOwnedGameplayTags (Variables|GetInstigatedBy) :!class /Script/GameplayTags.BlueprintGameplayTagLibrary) (GameplayTags|GetOwnedGameplayTags (Utilities|Casting|ToGameplayTagAssetInterface (Variables|GetDamagedActor)) :!class /Script/GameplayTags.BlueprintGameplayTagLibrary) (Variables|GetDamage) (Ability|TargetData|AbilityTargetDatafromHitResult (Variables|GetHitInfo)))))

(fn GetAnimMontageForHitReactionFromGameplayEventDataAndProxyAsset
  (params (ActorToCheckForAnimContent :type Actor%20Object%20Reference) (ProxyAsset :type Proxy%20Asset%20Object%20Reference) (GameplayEventData :type Gameplay%20Event%20Data%20Structure))
  (Variables|GetProxyAsset
    (:then
      (bind _actortocheckforanimcontent (Variables|GetActorToCheckForAnimContent)
        (:then
          (bind _asfort_pawn (Utilities|Casting|CastToFortPawn _actortocheckforanimcontent))
          (bind _returnvalue (Class|FortPawnHitReactionModuleBlueprintLibrary|GetAnimMontageforHitReaction _asfort_pawn (Variables|GetGameplayEventData) (Variables|GetProxyAsset)))
          (Utilities|IsValid (Utilities|Struct|BreakFortPawnHitReactionEvaluationOutput _returnvalue)
            (:Is%20Valid
              (return true (Utilities|Struct|BreakFortPawnHitReactionEvaluationOutput _returnvalue) (.SectionName (Utilities|Struct|BreakFortPawnHitReactionEvaluationOutput _returnvalue)) (.AnimRootMotionScale (Utilities|Struct|BreakFortPawnHitReactionEvaluationOutput _returnvalue)) (.bIsInterruptive (Utilities|Struct|BreakFortPawnHitReactionEvaluationOutput _returnvalue))))
            (:Is%20Not%20Valid
              (return false 0 "None" 0.0 true))) ; We found Montage Content to use?
        )
        (:else)) ; This is likely our Pawn and we need to check them \nfor the Proxy Table that contains the Proxy Asset \nthat we are looking for.
    )
    (:else)) ; Do we have Valid Data for this \nto attempt to look for?
)

(fn GetAnimMontageFromProxyAsset
  (params (ActorToCheckForAnimContent :type Actor%20Object%20Reference) (ProxyAsset :type Proxy%20Asset%20Object%20Reference))
  (bind _proxyasset (Variables|GetProxyAsset))
  (Variables|GetProxyAsset
    (:then
      (bind _actortocheckforanimcontent (Variables|GetActorToCheckForAnimContent)
        (:then
          (bind (_has_fort_ai_animation_component _fort_ai_animation_component) (CallFunction|ActorHasFortAIAnimationComponent _actortocheckforanimcontent))
          (if _has_fort_ai_animation_component
            (Utilities|IsValid _fort_ai_animation_component
              (:Is%20Valid
                (Utilities|IsValid (Class|FortAIComponentAnimation|GetAnimMontage _fort_ai_animation_component _proxyasset)
                  (:Is%20Valid
                    (return true (Class|FortAIComponentAnimation|GetAnimMontage _fort_ai_animation_component _proxyasset)))
                  (:Is%20Not%20Valid
                    (return false 0))) ; We found Montage Content to use?
              ))))
        (:else)) ; This is likely our Pawn and we need to check them \nfor the Proxy Table that contains the Proxy Asset \nthat we are looking for.
    )
    (:else)) ; Do we have Valid Data for this \nto attempt to look for?
)

(fn HitIsKnockback
  (params (Effect%20Context :type Gameplay%20Effect%20Context%20Handle%20Structure) (FortPawn :type Fort%20Pawn%20Object%20Reference))
  (bind (_impulsedirection _knockbackmagnitude) (Ability|EffectContext|GetImpulseData (Variables|GetEffect%20Context)))
  (Utilities|FlowControl|Sequence
    (:then_0
      (bind _fortpawn (Variables|GetFortPawn)
        (:then
          (return (> _knockbackmagnitude (FortPawn|GetKnockbackThreshold _fortpawn))))
        (:else)))
    (:then_1
      (return false))))

(fn IsFortPawnBeingControlled
  (params (FortPawn :type Fort%20Pawn%20Object%20Reference))
  (return (Math|Boolean|ORBoolean (Variables|/Script/FortniteGame.FortPawn|GetbIsStaggered FortPawn) (Variables|/Script/FortniteGame.FortPawn|GetbIsKnockedback FortPawn) (Variables|/Script/FortniteGame.FortPawn|GetbIsStunned FortPawn)) (Variables|/Script/FortniteGame.FortPawn|GetbIsStaggered FortPawn) (Variables|/Script/FortniteGame.FortPawn|GetbIsKnockedback FortPawn) (Variables|/Script/FortniteGame.FortPawn|GetbIsStunned FortPawn)))

(fn IsFortPawnControlResistanceFull
  (params (FortPawn :type Fort%20Pawn%20Object%20Reference))
  (return (== (FortPawn|GetMaxControlResistance FortPawn) (FortPawn|GetControlResistance FortPawn))))

(fn IsFortPawnFullyRecoveredFromControl
  (params (FortPawn :type Fort%20Pawn%20Object%20Reference))
  (return (and (CallFunction|IsFortPawnControlResistanceFull FortPawn) (not (CallFunction|IsFortPawnBeingControlled FortPawn)))))

(fn PlayAdditiveHitReactionLocalMontage
  (params (Montage :type Anim%20Montage%20Object%20Reference) (Section :type Name) (FortPawn :type Fort%20Pawn%20Object%20Reference))
  (Utilities|FlowControl|Sequence
    (:then_0
      (Variables|GetFortPawn
        (:then
          (Variables|GetMontage
            (:then
              (Animation|PlayLocalAnimMontage (Variables|GetFortPawn) (Variables|GetMontage) 1.0 (Variables|GetSection) false)
              (return true (Utilities|Time|GetGameTimeinSeconds 0)))
            (:else)))
        (:else)))
    (:then_1
      (return false -1.0))))
