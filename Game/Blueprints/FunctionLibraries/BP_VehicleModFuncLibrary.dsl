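; Blueprint function library that computes aim, firing and activation animation values for vehicle mod turrets.
; Seated pawn aim is converted to vehicle or attachment space, and pitch and yaw are normalized by 360 degrees.
; Setters write rotation and time-since-fire progress into static mesh Custom Primitive Data, while activation progress is only returned.
; The smoothed rotation variant optionally lerps from caller-held previous angles and is marked WIP, not finished.
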(fn BP_UpdateGetAdjustedRotations
  (params (InOwnerVehicle :type Fort%20Athena%20Vehicle%20Object%20Reference) (InTurretSeatIndex :type Integer) (BaseAtttachmentTransform :type Transform))
  (bind _inturretseatindex (Variables|GetInTurretSeatIndex))
  (bind _inownervehicle (Variables|GetInOwnerVehicle))
  (bind _returnvalue (Vehicle|GetPawnatSeat _inownervehicle _inturretseatindex))
  (Utilities|IsValid _returnvalue
    (:Is%20Valid
      (return (* (Math|Float|Clamp(Float) (/ (.pitch (Math|Transform|InverseTransformRotation BaseAtttachmentTransform (Pawn|GetBaseAimRotation _returnvalue))) 3.6e+02) -1.0) -1.0) (Math|Float|Clamp(Float) (/ (.yaw (Math|Transform|InverseTransformRotation BaseAtttachmentTransform (Pawn|GetBaseAimRotation _returnvalue))) 3.6e+02) -1.0)))))

(fn BP_UpdateGetFiringAnimationValue
  (params (InOwnerVehicle :type Fort%20Athena%20Vehicle%20Object%20Reference) (InLastFireTimeStamp :type Float%20%28double-precision%29) (InBarrelFireAnimDurationSeconds :type Float%20%28double-precision%29))
  (return (Math|Float|Clamp(Float) (Math|Float|SafeDivide (- (Utilities|Time|GetGameTimeinSeconds (Variables|GetInOwnerVehicle)) (Variables|GetInLastFireTimeStamp)) (Variables|GetInBarrelFireAnimDurationSeconds)))))

(fn BP_UpdateSetRotationPrimIndexValues
  (params (InTurretStaticMesh :type Static%20Mesh%20Component%20Object%20Reference) (InPrimitveDataIndexPitch :type Integer) (InPrimitveDataIndexYaw :type Integer) (InPitch :type Float%20%28double-precision%29) (InYaw :type Float%20%28double-precision%29))
  (Rendering|Material|SetCustomPrimitiveDataFloat (Variables|GetInTurretStaticMesh) (Variables|GetInPrimitveDataIndexPitch) (Variables|GetInPitch))
  (Rendering|Material|SetCustomPrimitiveDataFloat (Variables|GetInTurretStaticMesh) (Variables|GetInPrimitveDataIndexYaw) (Variables|GetInYaw)))

(fn BP_UpdateTurretStaticMesh_ActivationPercent
  (params (TurretActive :type Boolean) (TurretActivationProgressPerc :type Float%20%28double-precision%29) (UpdateStaticMeshFrequency :type Float%20%28double-precision%29) (TurretActivationAnimDurationSeconds :type Float%20%28double-precision%29))
  (return (Math|Float|Clamp(Float) (+ TurretActivationProgressPerc (* (/ UpdateStaticMeshFrequency TurretActivationAnimDurationSeconds) (Math|Float|SelectFloat 1.0 -1.0 TurretActive))))))

(fn BP_UpdateVehicleModTurretPrimitiveData_FireAnimation
  (params (InOwnerVehicle :type Fort%20Athena%20Vehicle%20Object%20Reference) (InTurretStaticMeshComp :type Static%20Mesh%20Component%20Object%20Reference) (InLastFireTimeStamp :type Float%20%28double-precision%29) (InBarrelFireAnimDurationSeconds :type Float%20%28double-precision%29) (PrimitiveDataIndex :type Integer :default 1))
  (Utilities|FlowControl|Sequence
    (:then_1
      (Rendering|Material|SetCustomPrimitiveDataFloat (Variables|GetInTurretStaticMeshComp) (Variables|GetPrimitiveDataIndex) (Math|Float|Clamp(Float) (Math|Float|SafeDivide (- (Utilities|Time|GetGameTimeinSeconds (Variables|GetInOwnerVehicle)) (Variables|GetInLastFireTimeStamp)) (Variables|GetInBarrelFireAnimDurationSeconds))))
      (return))))

(fn BP_UpdateVehicleModTurretPrimitiveData_Rotation
  (params (InOwnerVehicle :type Fort%20Athena%20Vehicle%20Object%20Reference) (InTurretStaticMeshComp :type Static%20Mesh%20Component%20Object%20Reference) (InTurretSeatIndex :type Integer) (InPrimitiveDataIndex_Pitch :type Integer) (InPrimitiveDataIndex_Yaw :type Integer :default 2) (InSocketPitchOffset :type Float%20%28double-precision%29) (InSocketYawOffset :type Float%20%28double-precision%29))
  (bind _inturretseatindex (Variables|GetInTurretSeatIndex))
  (bind _inownervehicle (Variables|GetInOwnerVehicle))
  (bind _returnvalue (Vehicle|GetPawnatSeat _inownervehicle _inturretseatindex))
  (bind _returnvalue_2 (Pawn|GetBaseAimRotation _returnvalue))
  (bind _inownervehicle_2 (Variables|GetInOwnerVehicle))
  (bind _returnvalue_3 (Vehicle|GetVehicleRotation _inownervehicle_2))
  (bind _returnvalue_4 (Math|Conversions|ToTransform(Rotator) _returnvalue_3))
  (bind _returnvalue_5 (Math|Transform|InverseTransformRotation _returnvalue_4 _returnvalue_2))
  (Utilities|IsValid _returnvalue
    (:Is%20Valid
      (Utilities|FlowControl|Sequence
        (:then_0
          (Rendering|Material|SetCustomPrimitiveDataFloat (Variables|GetInTurretStaticMeshComp) (Variables|GetInPrimitiveDataIndex_Pitch) (* (Math|Float|Clamp(Float) (/ (+ (.pitch _returnvalue_5) (Variables|GetInSocketPitchOffset)) 3.6e+02) -1.0) -1.0))
          (Rendering|Material|SetCustomPrimitiveDataFloat (Variables|GetInTurretStaticMeshComp) (Variables|GetInPrimitiveDataIndex_Yaw) (Math|Float|Clamp(Float) (/ (+ (.yaw _returnvalue_5) (Variables|GetInSocketYawOffset)) 3.6e+02) -1.0)))))))

(fn BP_UpdateVehicleModTurretPrimitiveData_Rotation_Smoothed
  (params (InOwnerVehicle :type Fort%20Athena%20Vehicle%20Object%20Reference) (InTurretStaticMeshComp :type Static%20Mesh%20Component%20Object%20Reference) (InTurretSeatIndex :type Integer) (InPrimitiveDataIndex_Pitch :type Integer) (InPrimitiveDataIndex_Yaw :type Integer :default 2) (InSocketPitchOffset :type Float%20%28double-precision%29) (InSocketYawOffset :type Float%20%28double-precision%29) (UseLerp :type Boolean) (RotationSpeed :type Float%20%28double-precision%29) (InDeltaTime :type Float%20%28double-precision%29) (PreviousPtich :type Float%20%28double-precision%29) (PreviousYaw :type Float%20%28double-precision%29))
  (bind _inturretseatindex (Variables|GetInTurretSeatIndex))
  (bind _inownervehicle (Variables|GetInOwnerVehicle))
  (bind _returnvalue (Vehicle|GetPawnatSeat _inownervehicle _inturretseatindex))
  (bind _returnvalue_2 (Pawn|GetBaseAimRotation _returnvalue))
  (bind _rotationthisframe (LocalVariables|GetRotationThisFrame))
  (bind _returnvalue_3 (Math|Vector2D|MakeVector2D _rotationthisframe _rotationthisframe))
  (bind _previousyaw (Variables|GetPreviousYaw))
  (bind _previousptich (Variables|GetPreviousPtich))
  (bind _returnvalue_4 (Math|Vector2D|MakeVector2D _previousptich _previousyaw))
  (bind _outnewyaw (LocalVariables|GetOutNewYaw))
  (bind _outnewpitch (LocalVariables|GetOutNewPitch))
  (bind _returnvalue_5 (Math|Vector2D|MakeVector2D _outnewpitch _outnewyaw))
  (bind _returnvalue_6 (- _returnvalue_5 _returnvalue_4))
  (bind _returnvalue_7 (/ _returnvalue_6 _returnvalue_3))
  (bind _returnvalue_8 (Math|Vector2D|ClampAxes2D _returnvalue_7 -1.0 1.0))
  (bind _returnvalue_9 (Math|Vector2D|GetAbs2D _returnvalue_8))
  (bind (_x _y) (Math|Vector2D|BreakVector2D _returnvalue_9))
  (bind _inownervehicle_2 (Variables|GetInOwnerVehicle))
  (bind _returnvalue_10 (Vehicle|GetVehicleRotation _inownervehicle_2))
  (bind _returnvalue_11 (Math|Conversions|ToTransform(Rotator) _returnvalue_10))
  (bind _returnvalue_12 (Math|Transform|InverseTransformRotation _returnvalue_11 _returnvalue_2))
  ;[1 1 1] WIP NOT FINISHED
  (Utilities|FlowControl|Sequence
    (:then_0
      (LocalVariables|SetOutNewPitch PreviousPtich)
      (LocalVariables|SetOutNewYaw PreviousYaw)
      (Utilities|IsValid _returnvalue
        (:Is%20Valid
          (Utilities|FlowControl|Sequence
            (:then_0
              (LocalVariables|SetOutNewYaw (.yaw _returnvalue_12))
              (LocalVariables|SetOutNewPitch (.pitch _returnvalue_12))
              (if (Variables|GetUseLerp)
                (LocalVariables|SetRotationThisFrame (* (Variables|GetRotationSpeed) (Variables|GetInDeltaTime)))
                (LocalVariables|SetOutNewPitch (Math|Float|Lerp (Variables|GetPreviousPtich) (LocalVariables|GetOutNewPitch) _x))
                (LocalVariables|SetOutNewYaw (Math|Float|Lerp (Variables|GetPreviousYaw) (LocalVariables|GetOutNewYaw) _y))
                (else
                  _))
              (Rendering|Material|SetCustomPrimitiveDataFloat (Variables|GetInTurretStaticMeshComp) (Variables|GetInPrimitiveDataIndex_Pitch) (* (Math|Float|Clamp(Float) (/ (+ (LocalVariables|GetOutNewPitch) (Variables|GetInSocketPitchOffset)) 3.6e+02) -1.0) -1.0))
              (Rendering|Material|SetCustomPrimitiveDataFloat (Variables|GetInTurretStaticMeshComp) (Variables|GetInPrimitiveDataIndex_Yaw) (Math|Float|Clamp(Float) (/ (+ (LocalVariables|GetOutNewYaw) (Variables|GetInSocketYawOffset)) 3.6e+02) -1.0)))))))
    (:then_1
      (return (LocalVariables|GetOutNewPitch) (LocalVariables|GetOutNewYaw)))))
