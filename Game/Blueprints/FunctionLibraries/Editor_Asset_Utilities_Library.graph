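# Editor utility function library with helpers for texture assets and content paths.
# GetOrCreateTexture2D_Util joins package path and asset name, loads the Texture2D if it exists or creates one through Asset Tools, then returns it.
# StripRightMostFolderFromPath_Util splits a path on slashes and rebuilds it until Strip Count segments remain, returning the shortened path and the segment at the cut.
# Both print debug strings; an asset-delete node is left unconnected.
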EditorUtilityBlueprint /Game/Blueprints/FunctionLibraries/Editor_Asset_Utilities_Library
graphs [
 GetOrCreateTexture2D_Util {nodes [
   n1 = GetOrCreateTexture2D_Util!() -> (then, PackagePath:String, Asset Name:String, WorldContextObject:Object Reference) {is=title K2Node_FunctionEntry}
   > n2 = Sequence()
   > PrintString(WorldContextObject=@n1.WorldContextObject, InString=@n1)
   n3 = Append!(A=@n1, B=/, C:String=@n1.Asset Name, D:String) -> (ReturnValue)
   > n4 = DoesAssetExist(AssetPath=@n3)
   > n5 = Branch(Condition=@n4)
   > n6 = LoadAsset(AssetPath=@n3)
   > n7 = Cast To Texture2D(Object=@n6)
   > Branch(Condition=IsValid(Object=@n6))
   > Set Texture_R(Texture_R:Texture 2D Object Reference=@n7) -> (Output_Get:Texture 2D Object Reference)
   n8 = GetAssetTools(execute=@n5.else)
   > n9 = CreateAsset(self=@n8, AssetName=@n1.Asset Name, PackagePath=@n1, AssetClass=/Script/Engine.Texture2D)
   > n10 = Cast To Texture2D(Object=@n9)
   > PrintString(WorldContextObject=@n1.WorldContextObject, InString=GetDisplayName(Object=@n9))
   > Set Texture_R(Texture_R=@n10)
   Return Node!(execute=@n2.then_1, Texture:Texture 2D Object Reference=Get Texture_R() -> (Texture_R:Texture 2D Object Reference))
   DeleteLoadedAsset!(execute, AssetToDelete:Object Reference=@n7) -> (then, ReturnValue:bool)
  ]}
 StripRightMostFolderFromPath_Util {nodes [
   n1 = StripRightMostFolderFromPath_Util!() -> (then, Path:String, WorldContextObject:Object Reference, Strip Count:int=1) {is=title K2Node_FunctionEntry}
   > n2 = Sequence()
   n3 = ParseIntoArray(SourceString=@n1, Delimiter=/)
   > n4 = For Each Loop with Break!(Exec:Exec, Array:Array of Strings=@n3, Break:Exec=@n7.then) -> (LoopBody:Exec, Array Element:String, Array Index:int, Completed:Exec)
   n5 = Length(TargetArray:Array of Strings=@n3)
   > n6 = Branch(Condition="Equal (Integer)"(A=@n4.Array Index, B=int - int(A=@n5, B=@n1.Strip Count)))
   > n7 = Set LastString(LastString:String=@n4) -> (Output_Get:String)
   n8 = Get FinalPath() -> (FinalPath:String)
   Set FinalPath(execute=@n6.else, FinalPath:String=Append!(A=Append(A=@n8, B=/), B=@n4) -> (ReturnValue)) -> (Output_Get:String)
   PrintString(execute=@n2.then_1, WorldContextObject=@n1.WorldContextObject, InString=Get FinalPath())
   > Return Node!(execute, FinalPath1:String=Get FinalPath(), AssetName:String=Get LastString() -> (LastString:String))
   GetEditorWorld(self=Get UnrealEditorSubsystem())
  ]}
]
variables []
