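; Environment art function library that derives repeatable pseudo-random values from a world location.
; randomFloatByLocation divides the location by an interval, sums absolute X, Y and optionally Z, truncates, squares it and seeds a random stream, so locations with equal truncated sums share values.
; randomSteppedIntByLocation scales that float by stepCount minus one and rounds it, returning both integer and float.
; A stepCount of one or less rounds the unscaled float instead.
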(fn randomFloatByLocation
  (params (WorldLocation :type Vector) (distanceInterval :type Float%20%28double-precision%29 :default 2e+02) (includeZ :type Boolean))
  (return (Math|Random|RandomFloatfromStream (Math|Random|MakeRandomStream (* (Math|Float|Truncate (select includeZ (Utilities|Operators|Add (Math|Float|Absolute(Float) (.x (/ WorldLocation distanceInterval))) (Math|Float|Absolute(Float) (.y (/ WorldLocation distanceInterval))) (Math|Float|Absolute(Float) (.z (/ WorldLocation distanceInterval)))) (+ (Math|Float|Absolute(Float) (.x (/ WorldLocation distanceInterval))) (Math|Float|Absolute(Float) (.y (/ WorldLocation distanceInterval)))))) (Math|Float|Truncate (select includeZ (Utilities|Operators|Add (Math|Float|Absolute(Float) (.x (/ WorldLocation distanceInterval))) (Math|Float|Absolute(Float) (.y (/ WorldLocation distanceInterval))) (Math|Float|Absolute(Float) (.z (/ WorldLocation distanceInterval)))) (+ (Math|Float|Absolute(Float) (.x (/ WorldLocation distanceInterval))) (Math|Float|Absolute(Float) (.y (/ WorldLocation distanceInterval)))))))))))

(fn randomSteppedIntByLocation
  (params (WorldLocation :type Vector) (distanceInterval :type Float%20%28double-precision%29 :default 2e+02) (includeZ :type Boolean) (stepCount :type Integer :default 1))
  (LocalVariables|Setl_max (- stepCount 1))
  (LocalVariables|Setl_float (CallFunction|RandomFloatbyLocation WorldLocation distanceInterval includeZ))
  (if (> (LocalVariables|Getl_max) 0)
    (return (Math|Float|Round (* (LocalVariables|Getl_float) (LocalVariables|Getl_max))) (LocalVariables|Getl_float))
    (else
      (return (Math|Float|Round (LocalVariables|Getl_float)) (LocalVariables|Getl_float)))))
