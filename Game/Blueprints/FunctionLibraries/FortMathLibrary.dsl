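; FortMathLibrary is a Blueprint function library with a single math helper, FloorVector.
; FloorVector rounds each of a vector's X, Y and Z components down with Floor, converts the integer results back to floats and rebuilds them into a new vector.
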(fn FloorVector
  (params (Vector :type Vector))
  (return (Math|Vector|MakeVector (Math|Conversions|ToFloat(Integer) (Math|Float|Floor (.x Vector))) (Math|Conversions|ToFloat(Integer) (Math|Float|Floor (.y Vector))) (Math|Conversions|ToFloat(Integer) (Math|Float|Floor (.z Vector))))))
