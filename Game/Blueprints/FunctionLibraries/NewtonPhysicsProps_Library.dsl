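; Blueprint function library that identifies physics props and launches them with impulses on physically simulated islands.
; IsPhysicProp accepts Fort physics pawn objects or actors active in the physics simulation; both launch functions require active simulation and apply impulse to the root primitive.
; LaunchPhysicsProps linearly maps clamped mass to impulse strength along a direction, adding current downward speed to Z.
; LaunchPhysicsPropsWithVector applies the supplied impulse vector directly.
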(fn IsPhysicProp
  (params (ObjectToEvaluate :type Actor%20Object%20Reference))
  (bind (_asfort_physics_pawn_object _bsuccess) (Utilities|Casting|CastToFortPhysicsPawnObject ObjectToEvaluate))
  (bind _returnvalue (FortPhysicsSimulationLibrary|IsPhysicsObjectActiveinSimulation ObjectToEvaluate)) ; True if actor is a vehicle or has an active physics object interface.\nAlready checked vehicles, so that part doesn't matter.
  (return (or _returnvalue _bsuccess)))

(fn LaunchPhysicsProps
  (params (WorldContextObject :type Object%20Reference) (PhysicsPropToLaunch :type Actor%20Object%20Reference) (NormalizedLaunchDirection :type Vector) (MinPhysicsPropMass :type Float%20%28double-precision%29) (MaxPhysicsPropMass :type Float%20%28double-precision%29) (MinPhysicsPropImpulse :type Float%20%28double-precision%29) (MaxPhysicsPropImpulse :type Float%20%28double-precision%29))
  (bind (_returnvalue_x _returnvalue_y _returnvalue_z) (Math|Vector|BreakVector (Transformation|GetVelocity (Variables|GetPhysicsPropToLaunch))))
  (bind _returnvalue (FortPhysicsSimulationLibrary|IsPhysicsObjectActiveinSimulation PhysicsPropToLaunch))
  (if (and _returnvalue (Fortnite|IsPhysicallySimulatedIsland WorldContextObject))
    (bind _asprimitive_component (Utilities|Casting|CastToPrimitiveComponent (Variables|/Script/Engine.Actor|GetRootComponent PhysicsPropToLaunch)))
    (bind (_returnvalue_x_2 _returnvalue_y_2 _returnvalue_z_2) (Math|Vector|BreakVector (* (Variables|GetNormalizedLaunchDirection) (+ (* (/ (- (Math|Float|Clamp(Float) (Physics|GetMass _asprimitive_component) (Variables|GetMinPhysicsPropMass) (Variables|GetMaxPhysicsPropMass)) (Variables|GetMinPhysicsPropMass)) (- (Variables|GetMaxPhysicsPropMass) (Variables|GetMinPhysicsPropMass))) (- (Variables|GetMaxPhysicsPropImpulse) (Variables|GetMinPhysicsPropImpulse))) (Variables|GetMinPhysicsPropImpulse)))))
    (Physics|AddImpulse _asprimitive_component (Math|Vector|MakeVector _returnvalue_x_2 _returnvalue_y_2 (+ _returnvalue_z_2 (Math|Float|Absolute(Float) (Math|Float|Clamp(Float) _returnvalue_z -1e+05 0.0)))) "None" false)
    (return)))

(fn LaunchPhysicsPropsWithVector
  (params (WorldContextObject :type Object%20Reference) (PhysicsPropToLaunch :type Actor%20Object%20Reference) (ImpulseLaunchVector :type Vector))
  (bind _returnvalue (FortPhysicsSimulationLibrary|IsPhysicsObjectActiveinSimulation PhysicsPropToLaunch))
  (if (and _returnvalue (Fortnite|IsPhysicallySimulatedIsland WorldContextObject))
    (bind _asprimitive_component (Utilities|Casting|CastToPrimitiveComponent (Variables|/Script/Engine.Actor|GetRootComponent PhysicsPropToLaunch)))
    (Physics|AddImpulse _asprimitive_component (Variables|GetImpulseLaunchVector) "None" false)
    (return)))
