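; Function library of platform helpers that classify screen orientation, console platform and performance mode.
; Codename checks compare the platform name string: isDamascus matches PS5, isMithril matches XSX, and isJanus accepts either one or the material quality level 3.
; DisplayOrientationFromScreenOrientation collapses screen orientation values to Portrait or Landscape, treating unknown, face-up, face-down and full-sensor as Landscape.
; isPcPerfMode returns the performance mode flag from Fort game user settings, without a platform check.
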(fn DisplayOrientationFromScreenOrientation
  (params (ScreenOrientation :type EScreenOrientation%20Enum :default "Unknown"))
  (return (Utilities|Select :Index ScreenOrientation :Unknown "Landscape" :Portrait "Portrait" :PortraitUpsideDown "Portrait" :LandscapeLeft "Landscape" :LandscapeRight "Landscape" :FaceUp "Landscape" :FaceDown "Landscape" :PortraitSensor "Portrait" :LandscapeSensor "Landscape" :FullSensor "Landscape")))

(fn isDamascus
  (return (Utilities|String|EqualExactly(String) (Game|GetPlatformName) "PS5")))

(fn isJanus
  (return (Math|Boolean|ORBoolean (Utilities|String|EqualExactly(String) (Game|GetPlatformName) "PS5") (Utilities|String|EqualExactly(String) (Game|GetPlatformName) "XSX") (== (Rendering|Material|GetRenderingMaterialQualityLevel) 3))))

(fn isMithril
  (return (Utilities|String|EqualExactly(String) (Game|GetPlatformName) "XSX")))

(fn isPcPerfMode
  (return (Class|FortGameUserSettings|IsinPerformanceMode (Settings|GetFortGameUserSettings))))
